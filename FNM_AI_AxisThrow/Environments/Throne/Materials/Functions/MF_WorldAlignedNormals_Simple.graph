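# Material function that wraps the engine's WorldAlignedNormals_HighQuality to project a normal map in world space.
# It derives X and Z blend alphas from the absolute world normal with CheapContrast, and a negative divisor from texture size, both with static-switched input overrides.
# Static switches output the XYZ, flat-top, XY and Z projections in world space by default, or transformed to tangent space.
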MaterialFunction /FNM_AI_AxisThrow/Environments/Throne/Materials/Functions/MF_WorldAlignedNormals_Simple
graph {nodes [
  n1 = MaterialFunctionCall!(
   Negative World Position Divisor=NamedRerouteUsage()
   World Position=NamedRerouteUsage()
   Texture=FunctionInput(Preview=TextureObject() {samplerType=SAMPLERTYPE_Normal, texture=/Engine/EngineMaterials/FlatNormal}) {bUsePreviewValueAsDefault=true, inputName=NormalTextureObject, inputType=FunctionInput_Texture2D, sortPriority=30}
   Z lerp Alpha=NamedRerouteUsage()
   WorldNormal=NamedRerouteUsage()
   X lerp alpha=NamedRerouteUsage()
  ) {materialFunction=/Engine/Functions/Engine_MaterialFunctions02/WorldAlignedNormals_HighQuality}
  n2 = FunctionInput(Preview=StaticBool()) {
   bUsePreviewValueAsDefault=true
   description="This switch defaults to true. If left true ensure that tangent space normals is turned off in the shaders properties panel. "
   inputName=World Space Normals
   inputType=FunctionInput_StaticBool
   sortPriority=31
  }
  FunctionOutput(None=StaticSwitch(
    True=@n1.World Space Normals XYZ
    False=Transform(None=@n1.World Space Normals XYZ) {transformSourceType=TRANSFORMSOURCE_World, transformType=TRANSFORM_Tangent}
    Value=@n2
   )) {outputName=Normal XYZ Texture}
  FunctionOutput(None=StaticSwitch(
    True=@n1.Z Projection World Space Normals
    False=Transform(None=@n1.Z Projection World Space Normals) {transformSourceType=TRANSFORMSOURCE_World, transformType=TRANSFORM_Tangent}
    Value=@n2
   )) {outputName=Normal Z Texture, sortPriority=3}
  FunctionOutput(None=StaticSwitch(
    True=@n1.XY Projection World Space Normals
    False=Transform(None=@n1.XY Projection World Space Normals) {transformSourceType=TRANSFORMSOURCE_World, transformType=TRANSFORM_Tangent}
    Value=@n2
   )) {outputName=Normal XY Texture, sortPriority=2}
  FunctionOutput(None=StaticSwitch(
    True=@n1.XY Flat Top Projection World Space Normals
    False=Transform(None=@n1.XY Flat Top Projection World Space Normals) {transformSourceType=TRANSFORMSOURCE_World, transformType=TRANSFORM_Tangent}
    Value=@n2
   )) {outputName=Normal XYZ Flat Top, sortPriority=1}
  n3 = FunctionInput(Preview=VertexNormalWS()) {bUsePreviewValueAsDefault=true, inputName=World Space Normal, sortPriority=201, previewValue=<0 0 1 1>}
  n4 = NamedRerouteDeclaration(None=@n3) {name=WSNormals, nodeColor=<rgba 1 0.5156 0 1>}
  n5 = Abs(None=ComponentMask(None=@n4) {r=true})
  n6 = FunctionInput() {bUsePreviewValueAsDefault=true, inputName=ProjectionTransitionContrast, inputType=FunctionInput_Scalar, sortPriority=7}
  NamedRerouteDeclaration(None=ComponentMask(None=MaterialFunctionCall(In=@n5, Contrast=@n6) {materialFunction=&CheapContrast}.Result) {desc="1=+x", r=true}) {name=Def_XLerpAlpha, nodeColor=<rgba 1 0 0.1172 1>}
  n7 = Abs(None=ComponentMask(None=@n4) {b=true})
  NamedRerouteDeclaration(None=ComponentMask(None=MaterialFunctionCall(In=@n7, Contrast=@n6) {materialFunction=&CheapContrast}.Result) {desc="1=+z", r=true}) {name=Def_ZLerpAlpha, nodeColor=<rgba 0 0.4375 1 1>}
  n8 = AppendVector(A=NamedRerouteUsage(), B=NamedRerouteUsage())
  n9 = FunctionInput(Preview=StaticBool()) {bUsePreviewValueAsDefault=true, inputName=UseInputXZLerp, inputType=FunctionInput_StaticBool, sortPriority=6}
  n10 = MaterialFunctionCall(Float2=StaticSwitch(
    True=FunctionInput(Preview=@n8) {bUsePreviewValueAsDefault=true, inputName=XZLerp, inputType=FunctionInput_Vector2}
    False=@n8
    Value=NamedRerouteDeclaration(None=@n9) {name=useInputs, nodeColor=<rgba 1 0 0.0703 1>}
   )) {materialFunction=&BreakOutFloat2Components}
  NamedRerouteDeclaration(None=@n10.G) {name=ZLerpAlpha, nodeColor=<rgba 0.125 1 0 1>}
  n11 = FunctionInput() {
   bUsePreviewValueAsDefault=true
   description=Size of texture in worldspace units
   inputName=TextureSize
   sortPriority=-11
   previewValue=<64 64 64 6>
  }
  n12 = NamedRerouteDeclaration(None=Multiply(A=Abs(None=@n11)) {constB=-1}) {name=Def_NWPD, nodeColor=<rgba 1 0 0.8437 1>}
  NamedRerouteDeclaration(None=StaticSwitch(True=FunctionInput(Preview=@n12) {bUsePreviewValueAsDefault=true, inputName=NegativeDivisor}, False=@n12, Value=NamedRerouteUsage())) {name=NWPD, nodeColor=<rgba 1 0 0.4453 1>}
  NamedRerouteDeclaration(None=FunctionInput(Preview=WorldPosition().XYZ) {bUsePreviewValueAsDefault=true, inputName=WorldPosition, sortPriority=200}) {name=WP, nodeColor=<rgba 0 1 0.8359 1>}
  NamedRerouteDeclaration(None=@n10.R) {name=XLerpAlpha, nodeColor=<rgba 0.1953 0 1 1>}
 ]}
