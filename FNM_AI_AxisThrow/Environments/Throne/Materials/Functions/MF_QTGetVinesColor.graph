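# Material function for the Throne environment's vines color that derives leg masking values from an inverted-V UV and a packed UV min/max vector.
# If nodes build a U-range mask, multiplied by saturated V, which a static switch takes raw for curve alpha or remapped.
# The remap feeds the V minimum into both input bounds.
# It outputs this product as LegsCurvetime, and as LegsParam zeroed where V exceeds its maximum.
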MaterialFunction /FNM_AI_AxisThrow/Environments/Throne/Materials/Functions/MF_QTGetVinesColor
graph {nodes [
  n1 = MaterialFunctionCall(Float4=FunctionInput() {desc=MinMax U and V respectively, inputName=InLegsUVParams, inputType=FunctionInput_Vector4, sortPriority=1}) {materialFunction=&BreakOutFloat4Components}
  n2 = Constant()
  n3 = MaterialFunctionCall(Float2=FunctionInput() {desc=V is inverted 1-V, inputName=InLegsUV, inputType=FunctionInput_Vector2}) {materialFunction=&BreakOutFloat2Components}
  n4 = Constant() {r=1}
  n5 = MaterialFunctionCall(Input=@n3.G, Input Low=@n1.B, Input High=@n1.B) {materialFunction=&RemapValueRange}
  n6 = FunctionInput(Preview=StaticBool()) {
   bUsePreviewValueAsDefault=true
   desc=No need to Remap if using Curve Alpha as param
   inputName=InUsesCurvesAlpha
   inputType=FunctionInput_StaticBool
   sortPriority=2
  }
  n7 = Multiply(
   A=Multiply(A=If(A=@n3.R, B=@n1.R, "A > B"=@n4, "A == B"=@n4, "A < B"=@n2), B=If(A=@n3.R, B=@n1.G, "A > B"=@n2, "A == B"=@n4, "A < B"=@n4))
   B=Saturate(None=StaticSwitch(True=@n3.G, False=@n5.Result, Value=@n6))
  )
  FunctionOutput(None=@n7) {outputName=LegsCurvetime}
  FunctionOutput(None=If(A=@n3.G, B=@n1.A, "A > B"=Constant(), "A == B"=@n7, "A < B"=@n7)) {outputName=LegsParam, sortPriority=1}
 ]}
