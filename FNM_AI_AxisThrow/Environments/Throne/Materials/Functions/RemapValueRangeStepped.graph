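# RemapValueRangeStepped is a library-exposed material function in the Math category that linearly remaps a scalar from an input range to a target range.
# The Smooth output normalizes Input between Input Low and Input High, then scales and offsets it to Target Low and Target High, without clamping.
# The Stepped output floors that result to increments of 1/Steps, so step size is independent of the target range's width.
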MaterialFunction /FNM_AI_AxisThrow/Environments/Throne/Materials/Functions/RemapValueRangeStepped
graph {
 properties {
  bExposeToLibrary=true
  description=Generalized remapping function to explicitly remap an input range to an output range.
  libraryCategoriesText [Math]
 }
 nodes [
  n1 = FunctionInput() {bUsePreviewValueAsDefault=true, inputName=Input Low, inputType=FunctionInput_Scalar, sortPriority=1}
  n2 = Subtract(A=FunctionInput() {inputName=Input, inputType=FunctionInput_Scalar}, B=@n1)
  n3 = FunctionInput() {bUsePreviewValueAsDefault=true, inputName=Input High, inputType=FunctionInput_Scalar, sortPriority=2, previewValue=<1 0 0 1>}
  n4 = Subtract(A=@n3, B=@n1)
  n5 = FunctionInput() {bUsePreviewValueAsDefault=true, inputName=Target High, inputType=FunctionInput_Scalar, sortPriority=4, previewValue=<1 0 0 1>}
  n6 = FunctionInput() {bUsePreviewValueAsDefault=true, inputName=Target Low, inputType=FunctionInput_Scalar, sortPriority=3}
  n7 = Add(A=Multiply(A=Divide(A=@n2, B=@n4), B=Subtract(A=@n5, B=@n6)), B=@n6)
  FunctionOutput(None=@n7) {outputName=Smooth}
  n8 = FunctionInput() {bUsePreviewValueAsDefault=true, inputName=Steps, inputType=FunctionInput_Scalar, sortPriority=5, previewValue=<1 0 0 1>}
  n9 = Multiply(A=@n7, B=@n8)
  FunctionOutput(None=Divide(A=Floor(None=@n9), B=@n8)) {outputName=Stepped, sortPriority=1}
 ]
}
